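# Sound cue for completing a search of a junk pile loot container, borrowing a single garage shelves search-end wave.
# An Enveloper wraps the wave player, holding volume near full for about a second before fading it to silence.
# It uses prop-search attenuation, the loot-collect sound class and a reduced volume multiplier.
# A concurrency override caps simultaneous instances, stopping the farthest first.
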SoundCue /Game/Sounds/Foley_Loot/Containers/JunkPile/JunkPile_SearchComplete_Cue
root=Enveloper_1
nodes [
 {
  id=Enveloper_1
  children [WavePlayer_1]
  properties {pitchCurve {editorCurveData {keys [{time=0}]}}, volumeCurve {editorCurveData {keys [<at 0 1>,<at 0.9973 1.006>,<at 1.7122 0>]}}}
 }
 {id=WavePlayer_1, properties {soundWaveAssetPtr=/Game/Sounds/Foley_Loot/Containers/Shelves/Shelves_Garage_SearchEnd7}}
]
properties {
 attenuationSettings=/Game/Sounds/Attenuation/Fort_Prop_Search_Att
 bOverrideConcurrency=true
 maxDistance=1300
 soundClassObject=&Fort_Loot_Collect
 volumeMultiplier=0.45
 concurrencyOverrides {maxCount=4, resolutionRule=StopFarthestThenPreventNew}
}
