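# Sound cue that plays when a junk pile container search ends, classed as loot-collect audio.
# A mixer layers a "no more loot" UI sound with a borrowed garage-shelves search-end foley at slightly lower level.
# An enveloper holds the foley at full volume for about a second, then fades it out.
# Prop-search attenuation spatializes it, and a concurrency override limits simultaneous instances, dropping the farthest first.
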SoundCue /Game/Sounds/Foley_Loot/Containers/JunkPile/JunkPile_SearchEnd_Cue
root=Mixer_10
nodes [
 {
  id=Enveloper_0
  children [WavePlayer_105]
  properties {pitchCurve {editorCurveData {keys [{time=0}]}}, volumeCurve {editorCurveData {keys [<at 0 1>,<at 0.9972 1.006>,<at 1.7122 0>]}}}
 }
 {id=Mixer_10, children [WavePlayer_35, Enveloper_0], properties {inputVolume [1,0.9]}}
 {id=WavePlayer_105, properties {soundWaveAssetPtr=/Game/Sounds/Foley_Loot/Containers/Shelves/Shelves_Garage_SearchEnd7}}
 {id=WavePlayer_35, properties {soundWaveAssetPtr=/Game/Sounds/Fort_UI_Sounds/Gameplay/Fort_UI_NoMoreLoot}}
]
properties {
 attenuationSettings=/Game/Sounds/Attenuation/Fort_Prop_Search_Att
 bOverrideConcurrency=true
 maxDistance=1300
 soundClassObject=&Fort_Loot_Collect
 volumeMultiplier=0.47
 concurrencyOverrides {maxCount=4, resolutionRule=StopFarthestThenPreventNew}
}
