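# Sound cue for searching a junk pile container, built from reused wastebasket, desk and shelf-junk search foley.
# A Mixer layers a random wastebasket sound, a random desk sound and a fixed-delay random shelf-junk sound; an empty slot in the wastebasket Random can drop that layer.
# A root Modulator slightly varies pitch and volume.
# It uses prop-search attenuation, the loot-collect sound class and a concurrency override that prevents new instances at the limit.
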SoundCue /Game/Sounds/Foley_Loot/Containers/JunkPile/JunkPile_Search_Cue
root=Modulator_11
nodes [
 {id=Delay_24, children [Random_24], properties {delayMax=0.8, delayMin=0.8}}
 {id=Mixer_2, children [Random_21,Random_23,Delay_24], properties {inputVolume [0.8,0.9,0.9]}}
 {id=Modulator_11, children [Mixer_2], properties {pitchMax=1.07, pitchMin=0.92, volumeMax=1.07, volumeMin=0.92}}
 {id=Random_21, children [WavePlayer_91, WavePlayer_92, WavePlayer_93, null], properties {weights [1,1,1,1]}}
 {id=Random_23, children [WavePlayer_100, WavePlayer_101], properties {weights [1,1]}}
 {id=Random_24, children [WavePlayer_102, WavePlayer_103, WavePlayer_104], properties {weights [1,1,1]}}
 {id=WavePlayer_100, properties {soundWaveAssetPtr=/Game/Sounds/Foley_Loot/Containers/Desk/Desk_Search11}}
 {id=WavePlayer_101, properties {soundWaveAssetPtr=/Game/Sounds/Foley_Loot/Containers/Desk/Desk_Search10}}
 {id=WavePlayer_102, properties {soundWaveAssetPtr=/Game/Sounds/Foley_Loot/Containers/Shelves/Shelves_Junk_Search12}}
 {id=WavePlayer_103, properties {soundWaveAssetPtr=/Game/Sounds/Foley_Loot/Containers/Shelves/Shelves_Junk_Search10}}
 {id=WavePlayer_104, properties {soundWaveAssetPtr=/Game/Sounds/Foley_Loot/Containers/Shelves/Shelves_Junk_Search11}}
 {id=WavePlayer_91, properties {soundWaveAssetPtr=/Game/Sounds/Foley_Loot/Containers/Garbage/WasteBasket_Bedroom_Search4}}
 {id=WavePlayer_92, properties {soundWaveAssetPtr=/Game/Sounds/Foley_Loot/Containers/Garbage/WasteBasket_Bedroom_Search1}}
 {id=WavePlayer_93, properties {soundWaveAssetPtr=/Game/Sounds/Foley_Loot/Containers/Garbage/WasteBasket_Bedroom_Search2}}
]
properties {
 attenuationSettings=/Game/Sounds/Attenuation/Fort_Prop_Search_Att
 bOverrideConcurrency=true
 maxDistance=1300
 soundClassObject=&Fort_Loot_Collect
 volumeMultiplier=0.45
 concurrencyOverrides {resolutionRule=PreventNew}
}
