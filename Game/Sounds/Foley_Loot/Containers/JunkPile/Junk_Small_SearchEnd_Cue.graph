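# Sound cue for the end of searching a small junk pile container, combining junk foley with the no-more-loot UI sound.
# A random pick of four small junk search waves is faded out by an envelope, then given slight pitch and volume variation by a modulator.
# A mixer layers that with the UI wave at slightly reduced volume.
# It uses prop search attenuation and caps concurrency, stopping the farthest instance first.
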SoundCue /Game/Sounds/Foley_Loot/Containers/JunkPile/Junk_Small_SearchEnd_Cue
root=Mixer_2
nodes [
 {
  id=Enveloper_10
  children [Random_19]
  properties {pitchCurve {editorCurveData {keys [{time=0}]}}, volumeCurve {editorCurveData {keys [<at 0 1>,<at 0.7 0>]}}}
 }
 {id=Mixer_2, children [WavePlayer_13, Modulator_1], properties {inputVolume [0.8,1]}}
 {id=Modulator_1, children [Enveloper_10], properties {pitchMax=1.07, pitchMin=0.92, volumeMax=1.07, volumeMin=0.92}}
 {id=Random_19, children [WavePlayer_134, WavePlayer_135, WavePlayer_139, WavePlayer_140], properties {weights [1,1,1,1]}}
 {id=WavePlayer_13, properties {soundWaveAssetPtr=/Game/Sounds/Fort_UI_Sounds/Gameplay/Fort_UI_NoMoreLoot}}
 {id=WavePlayer_134, properties {soundWaveAssetPtr=&/Junk_Search_Small01}}
 {id=WavePlayer_135, properties {soundWaveAssetPtr=&/Junk_Search_Small02}}
 {id=WavePlayer_139, properties {soundWaveAssetPtr=&/Junk_Search_Small06}}
 {id=WavePlayer_140, properties {soundWaveAssetPtr=&/Junk_Search_Small07}}
]
properties {
 attenuationSettings=/Game/Sounds/Attenuation/Fort_Prop_Search_Att
 bOverrideConcurrency=true
 maxDistance=1300
 soundClassObject=&Fort_Loot_Collect
 volumeMultiplier=1
 concurrencyOverrides {maxCount=4, resolutionRule=StopFarthestThenPreventNew}
}
