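# Sound cue for searching a small junk pile container during looting, played with prop-search attenuation on the loot-collect sound class.
# A mixer layers two randomized search layers: one plays immediately, the other at slightly lower volume after a short random delay of under a second.
# Each layer picks from its own pool of junk search waves through a modulator that varies pitch and volume.
# Concurrency is overridden to prevent new instances once the limit is reached.
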SoundCue /Game/Sounds/Foley_Loot/Containers/JunkPile/Junk_Small_Search_Cue
root=Mixer_8
nodes [
 {id=Delay_25, children [Modulator_5], properties {delayMax=0.8, delayMin=0.75}}
 {id=Mixer_8, children [Modulator_4,Delay_25], properties {inputVolume [1,0.8]}}
 {id=Modulator_4, children [Random_10], properties {pitchMax=1.07, pitchMin=0.92, volumeMax=1.07, volumeMin=0.92}}
 {id=Modulator_5, children [Random_20], properties {pitchMax=1.07, pitchMin=0.92, volumeMax=1.07, volumeMin=0.92}}
 {id=Random_10, children [WavePlayer_142, WavePlayer_145, WavePlayer_146], properties {weights [1,1,1]}}
 {id=Random_20, children [WavePlayer_141, WavePlayer_147], properties {weights [1,1]}}
 {id=WavePlayer_141, properties {soundWaveAssetPtr=&/Junk_Search_Small01}}
 {id=WavePlayer_142, properties {soundWaveAssetPtr=&/Junk_Search_Small02}}
 {id=WavePlayer_145, properties {soundWaveAssetPtr=&/Junk_Search_Small05}}
 {id=WavePlayer_146, properties {soundWaveAssetPtr=&/Junk_Search_Small06}}
 {id=WavePlayer_147, properties {soundWaveAssetPtr=&/Junk_Search_Small07}}
]
properties {
 attenuationSettings=/Game/Sounds/Attenuation/Fort_Prop_Search_Att
 bOverrideConcurrency=true
 maxDistance=1300
 soundClassObject=&Fort_Loot_Collect
 volumeMultiplier=1.05
 concurrencyOverrides {resolutionRule=PreventNew}
}
